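; Creative island group playspace Blueprint that warns players about a pending device deprecation, apparently the time-of-day skydome.
; After a stopgap delay for island initialization, it asynchronously creates a UIKit dialog with a deprecation view model and pushes it as a modal for the local controller.
; The first button opts into the day sequence actor upgrade; the second teleports the player to the playground lobby island.
; Either choice then unbinds and closes the dialog.
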(event EventBeginPlay)

(event EventTick
  (params (DeltaSeconds :type Float%20%28single-precision%29)))

(event EventShowPendingDeviceDeprecationModal
  (params (FortPC :type Fort%20Player%20Controller%20Object%20Reference))
  (CallFunction|AlertPlayerOfPendingDeviceDeprecation FortPC))

(event Custom|AlertPlayerOfPendingDeviceDeprecation
  (params (FortPC :type Fort%20Player%20Controller%20Object%20Reference))
  (Utilities|FlowControl|Delay 3.0) ; TODO: This is a quick fix for FORT-973726.  We should move the ShowPendingDeviceDeprecationModal test later during island initialization.
  (if (Pawn|IsLocalController FortPC)
    (Variables|Default|SetPlayerController FortPC)
    (bind _userwidget (Class|FortAsyncActionCreateWidgetAsync|CreateWidgetAsync "/FortUILibrary/UIKit/Dialogs/WBP_UIKit_Dialog.WBP_UIKit_Dialog_C" FortPC false)
      (:OnComplete
        (bind _aswbp_uikit_dialog (Utilities|Casting|CastToWBP_UIKit_Dialog _userwidget))
        (bind _returnvalue (Game|ConstructObjectfromClass "/Game/Playgrounds/Playspaces/DialogVM_TODM_Skydome_Deprecation.DialogVM_TODM_Skydome_Deprecation_C"))
        (Viewmodel|SetUIKitDialogViewModel _aswbp_uikit_dialog _returnvalue :!class /FortUILibrary/UIKit/Dialogs/WBP_UIKit_Dialog.WBP_UIKit_Dialog_C)
        (Dispatchers|BindEventtoDialogButtonClickedDispatcher (Variables|/FortUILibrary/UIKit/Dialogs/WBP_UIKit_Dialog.WBP_UIKit_Dialog_C|GetDialog_Internal _aswbp_uikit_dialog) (delegate Custom|OnModalButtonClicked))
        (FortUIManagement|TryPushWidgetasModalPanelforPlayer FortPC _aswbp_uikit_dialog)))))

(event Custom|OnModalButtonClicked
  (params (ButtonIndex :type Integer))
  (switch int ButtonIndex
    (:0
      (Class|FortCreativeMyIslandUtils|OptinforDaySequenceActorUpgrade (Variables|Default|GetPlayerController)))
    (:1
      (bind _asfort_player_controller_athena (Utilities|Casting|CastToFortPlayerControllerAthena (Variables|Default|GetPlayerController))
        (:then
          (Class|FortPlayerControllerAthena|TeleporttoPlaygroundLobbyIsland _asfort_player_controller_athena))
        (:CastFailed
          _))))
  (Dispatchers|UnbindallEventsfromDialogButtonClickedDispatcher (Variables|/FortUILibrary/UIKit/Dialogs/WBP_UIKit_Dialog.WBP_UIKit_Dialog_C|GetDialog_Internal (stale Custom|AlertPlayerOfPendingDeviceDeprecation _aswbp_uikit_dialog)))
  (ActivatableWidget|DeactivateWidget (stale Custom|AlertPlayerOfPendingDeviceDeprecation _aswbp_uikit_dialog)))

(fn ConstructionScript)
